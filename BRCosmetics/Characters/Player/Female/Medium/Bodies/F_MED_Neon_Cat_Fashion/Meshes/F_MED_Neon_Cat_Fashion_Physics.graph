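# Physics asset for the Neon Cat Fashion female medium player body cosmetic, rigging only the dynamic jacket bones for secondary motion.
# Each sleeve has four small capsules constrained to a master capsule with tight swing and twist limits.
# The jacket's back, side and front panels each pair a capsule with a tiny master anchor body, allowing wider swing in some axes and very little twist.
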PhysicsAsset /BRCosmetics/Characters/Player/Female/Medium/Bodies/F_MED_Neon_Cat_Fashion/Meshes/F_MED_Neon_Cat_Fashion_Physics
bodies {
 Dyn_Jacket_Sleave1_l1 [{shapeType=Capsule, length=8, radius=1, center=<0.5 0 0>}]
 Dyn_Jacket_Sleave1_r [{shapeType=Capsule, length=8, radius=1, center=<-0.5 0 0>}]
 Dyn_Jacket_Sleave2_l1 [{shapeType=Capsule, length=8, radius=1, center=<0.5 0 0>, rotation=<rot 0 0 90>}]
 Dyn_Jacket_Sleave2_r [{shapeType=Capsule, length=8, radius=1, center=<-0.5 0 0>, rotation=<rot 0 0 90>}]
 Dyn_Jacket_Sleave3_1 [{shapeType=Capsule, length=8, radius=1, center=<0.5 0 0>, rotation=<rot 0 0 90>}]
 Dyn_Jacket_Sleave3_r [{shapeType=Capsule, length=8, radius=1, center=<-0.5 0 0>, rotation=<rot 0 0 90>}]
 Dyn_Jacket_Sleave4_l1 [{shapeType=Capsule, length=8, radius=1, center=<0.5 0 0>}]
 Dyn_Jacket_Sleave4_r [{shapeType=Capsule, length=8, radius=1, center=<-0.5 0 0>}]
 Dyn_Master_Jacket_Sleave_l [{shapeType=Capsule, length=8, radius=4, center=<0 1.7 0>, rotation=<rot 90 0 0>}]
 Dyn_Master__Sleave_r [{shapeType=Capsule, length=8, radius=4, center=<0 -1.72 0>, rotation=<rot 90 0 0>}]
 dyn_jacket_bk_01 [{shapeType=Capsule, length=5, radius=2, center=<2.919 0 0>, rotation=<rot 90 0 0>}]
 dyn_jacket_bk_MSTR [{shapeType=Capsule, length=0.5, radius=0.5}]
 dyn_jacket_bk_l_01 [{shapeType=Capsule, length=5, radius=2, center=<2.8821 -0.0001 -0.0001>, rotation=<rot 90 0 0>}]
 dyn_jacket_bk_l_MSTR [{shapeType=Capsule, length=0.5, radius=0.5}]
 dyn_jacket_bk_r_01 [{shapeType=Capsule, length=5, radius=2, center=<-2.8608 -0.0003 0>, rotation=<rot 90 0 0>}]
 dyn_jacket_bk_r_MSTR [{shapeType=Capsule, length=0.5, radius=0.5}]
 dyn_jacket_fr_l_01 [{shapeType=Capsule, length=5, radius=2, center=<2.8821 -0.0001 -0.0001>, rotation=<rot 90 0 0>}]
 dyn_jacket_fr_l_MSTR [{shapeType=Capsule, length=0.5, radius=0.5}]
 dyn_jacket_fr_r_01 [{shapeType=Capsule, length=5, radius=2, center=<-2.8608 -0.0003 0>, rotation=<rot 90 0 0>}]
 dyn_jacket_fr_r_MSTR [{shapeType=Capsule, length=0.5, radius=0.5}]
 dyn_jacket_l_01 [{shapeType=Capsule, length=5, radius=2, center=<2.8821 -0.0001 -0.0001>, rotation=<rot 90 0 0>}]
 dyn_jacket_l_MSTR [{shapeType=Capsule, length=0.5, radius=0.5}]
 dyn_jacket_r_01 [{shapeType=Capsule, length=5, radius=2, center=<-2.8608 -0.0003 0>, rotation=<rot 90 0 0>}]
 dyn_jacket_r_MSTR [{shapeType=Capsule, length=0.5, radius=0.5}]
}
constraints [
 {bone1Name=Dyn_Jacket_Sleave1_r, bone2Name=Dyn_Master__Sleave_r, swing1LimitDegrees=10, swing2LimitDegrees=10, twistLimitDegrees=1}
 {bone1Name=Dyn_Jacket_Sleave2_r, bone2Name=Dyn_Master__Sleave_r, swing1LimitDegrees=10, swing2LimitDegrees=10, twistLimitDegrees=1}
 {bone1Name=Dyn_Jacket_Sleave3_r, bone2Name=Dyn_Master__Sleave_r, swing1LimitDegrees=10, swing2LimitDegrees=10, twistLimitDegrees=1}
 {bone1Name=Dyn_Jacket_Sleave4_r, bone2Name=Dyn_Master__Sleave_r, swing1LimitDegrees=10, swing2LimitDegrees=10, twistLimitDegrees=1}
 {bone1Name=Dyn_Jacket_Sleave1_l1, bone2Name=Dyn_Master_Jacket_Sleave_l, swing1LimitDegrees=10, swing2LimitDegrees=10, twistLimitDegrees=1}
 {bone1Name=Dyn_Jacket_Sleave2_l1, bone2Name=Dyn_Master_Jacket_Sleave_l, swing1LimitDegrees=10, swing2LimitDegrees=10, twistLimitDegrees=1}
 {bone1Name=Dyn_Jacket_Sleave3_1, bone2Name=Dyn_Master_Jacket_Sleave_l, swing1LimitDegrees=10, swing2LimitDegrees=10, twistLimitDegrees=1}
 {bone1Name=Dyn_Jacket_Sleave4_l1, bone2Name=Dyn_Master_Jacket_Sleave_l, swing1LimitDegrees=10, swing2LimitDegrees=10, twistLimitDegrees=1}
 {bone1Name=dyn_jacket_bk_l_01, bone2Name=dyn_jacket_bk_l_MSTR, swing1LimitDegrees=40, swing2LimitDegrees=40, twistLimitDegrees=2}
 {bone1Name=dyn_jacket_fr_r_01, bone2Name=dyn_jacket_fr_r_MSTR, swing1LimitDegrees=10, swing2LimitDegrees=40, twistLimitDegrees=2}
 {bone1Name=dyn_jacket_fr_l_01, bone2Name=dyn_jacket_fr_l_MSTR, swing1LimitDegrees=10, swing2LimitDegrees=40, twistLimitDegrees=2}
 {bone1Name=dyn_jacket_l_01, bone2Name=dyn_jacket_l_MSTR, swing1LimitDegrees=40, swing2LimitDegrees=40, twistLimitDegrees=2}
 {bone1Name=dyn_jacket_bk_r_01, bone2Name=dyn_jacket_bk_r_MSTR, swing1LimitDegrees=40, swing2LimitDegrees=40, twistLimitDegrees=2}
 {bone1Name=dyn_jacket_bk_01, bone2Name=dyn_jacket_bk_MSTR, swing1LimitDegrees=10, swing2LimitDegrees=45, twistLimitDegrees=0.1}
 {bone1Name=dyn_jacket_r_01, bone2Name=dyn_jacket_r_MSTR, swing1LimitDegrees=40, swing2LimitDegrees=40, twistLimitDegrees=2}
]
