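# Post-process style AnimBlueprint for the Neon Cat Fashion female body, adding secondary motion to a pose copied from the attached parent mesh.
# Chained bone constraints blend glute translation between pelvis and thighs, and sleeve bones between hand and forearm twist bones.
# A RigidBody node then simulates the override physics asset in root-bone space with damped component motion.
# Two additive sleeve Transform Bone nodes are unconnected, and the event graph is empty.
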AnimBlueprint /BRCosmetics/Characters/Player/Female/Medium/Bodies/F_MED_Neon_Cat_Fashion/Meshes/F_MED_Neon_Cat_Fashion_AnimBP
target_skeleton=&/F_MED_Neon_Cat_Fashion_Skeleton
graphs [
 AnimGraph {nodes [
   "Transform (Modify) Bone - Bone: Dyn_Master__Sleave_r"(Alpha=1, Translation="0,0,0", Rotation="0,10,0", Scale="1,1,1") {
    rotationMode=BMM_Additive
    rotationSpace=BCS_BoneSpace
    translationMode=BMM_Additive
    translationSpace=BCS_BoneSpace
    alphaBoolBlend {blendOption=Linear}
    boneToModify {boneName=Dyn_Master__Sleave_r}
    rotation {yaw=10}
   }
   "Transform (Modify) Bone - Bone: Dyn_Master_Jacket_Sleave_l"(Alpha=1, Translation="0,0,0", Rotation="-20,0,-5", Scale="1,1,1") {
    rotationMode=BMM_Additive
    rotationSpace=BCS_BoneSpace
    translationMode=BMM_Additive
    translationSpace=BCS_BoneSpace
    alphaBoolBlend {blendOption=Linear}
    boneToModify {boneName=Dyn_Master_Jacket_Sleave_l}
    rotation {pitch=-20, roll=-5}
   }
   n1 = "Constraint - dyn_glute_r (TT)"!(
    ComponentPose=Local To Component(LocalPose=Copy Pose From Mesh() {bUseAttachedParent=true})
    Alpha=1
    ConstraintWeights_0=0.6
    ConstraintWeights_1:float=0.4
   ) -> (Pose) {
    alphaBoolBlend {blendOption=Linear}
    boneToModify {boneName=dyn_glute_r}
    constraintSetup [{transformType=Translation, targetBone {boneName=pelvis}}, {transformType=Translation, targetBone {boneName=thigh_r}}]
    constraintWeights [1,1]
   }
   n2 = "Constraint - dyn_glute_l (TT)"!(ComponentPose=@n1, Alpha=1, ConstraintWeights_0=0.6, ConstraintWeights_1:float=0.4) -> (Pose) {
    alphaBoolBlend {blendOption=Linear}
    boneToModify {boneName=dyn_glute_l}
    constraintSetup [{transformType=Translation, targetBone {boneName=pelvis}}, {transformType=Translation, targetBone {boneName=thigh_l}}]
    constraintWeights [1,1]
   }
   n3 = "Constraint - Dyn_Master__Sleave_r (RR)"!(ComponentPose=@n2, Alpha=1, ConstraintWeights_0=0.9, ConstraintWeights_1:float=0.1) -> (Pose) {
    alphaBoolBlend {blendOption=Linear}
    boneToModify {boneName=Dyn_Master__Sleave_r}
    constraintSetup [{transformType=Rotation, targetBone {boneName=hand_r}}, {transformType=Rotation, targetBone {boneName=lowerarm_twist_01_r}}]
    constraintWeights [1,1]
   }
   n4 = "Constraint - Dyn_Master_Jacket_Sleave_l (RR)"!(ComponentPose=@n3, Alpha=1, ConstraintWeights_0=0.1, ConstraintWeights_1:float=0.9) -> (Pose) {
    alphaBoolBlend {blendOption=Linear}
    boneToModify {boneName=Dyn_Master_Jacket_Sleave_l}
    constraintSetup [{transformType=Rotation, targetBone {boneName=hand_l}}, {transformType=Rotation, targetBone {boneName=lowerarm_twist_02_l}}]
    constraintWeights [1,1]
   }
   n5 = "Constraint - Dyn_Master_Jacket_Sleave_l (RR)"(ComponentPose=@n4, Alpha=1, ConstraintWeights_0=0.3, ConstraintWeights_1=0.7) {
    alphaBoolBlend {blendOption=Linear}
    boneToModify {boneName=Dyn_Master_Jacket_Sleave_l}
    constraintSetup [{transformType=Rotation, targetBone {boneName=hand_l}}, {transformType=Rotation, targetBone {boneName=lowerarm_twist_02_l}}]
    constraintWeights [1,1]
   }
   n6 = "Constraint - Dyn_Master_Jacket_Sleave_l (TT)"!(ComponentPose=@n5, Alpha=0.5, ConstraintWeights_0=0.9, ConstraintWeights_1:float=0.1) -> (Pose) {
    alphaBoolBlend {blendOption=Linear}
    boneToModify {boneName=Dyn_Master_Jacket_Sleave_l}
    constraintSetup [{transformType=Translation, targetBone {boneName=hand_l}}, {transformType=Translation, targetBone {boneName=lowerarm_twist_02_l}}]
    constraintWeights [1,1]
   }
   n7 = RigidBody(ComponentPose=@n6) {
    bClampLinearTranslationLimitToRefPose=true
    bFreezeIncomingPoseOnStart=true
    bOverrideWorldGravity=true
    overridePhysicsAsset=&/F_MED_Neon_Cat_Fashion_Physics
    simulationSpace=BaseBoneSpace
    alphaBoolBlend {blendOption=Linear}
    baseBoneRef {boneName=root}
    componentAppliedLinearAccClamp=<1200 1200 1200>
    componentLinearAccScale=<0.1 0.1 0.1>
    componentLinearVelScale=<0.1 0.1 0.1>
    overrideWorldGravity {z=-980}
   }
   Output Pose(Result=Component To Local(ComponentPose=@n7)) {layerGroup=None}
  ]}
 EventGraph {nodes [Event BlueprintUpdateAnimation()]}
]
variables []
