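# Physics asset for the back-hair ponytail of the Neon Cat Fashion female medium character cosmetic.
# A nine-capsule ponytail chain hangs from a tiny master body on the head, with tight twist limits and one swing axis loosening toward the tip.
# Capsules on the body, limbs and a backpack collision bone form free-rotating constraint links, apparently serving as hair colliders.
# An extra free constraint links the head to the second ponytail segment.
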PhysicsAsset /BRCosmetics/Characters/Player/Female/Medium/Bodies/F_MED_Neon_Cat_Fashion/Meshes/Parts/F_MED_Neon_Cat_Fashion_HairBack_Physics
bodies {
 calf_l [{shapeType=Capsule, length=35.255, radius=7.755, center=<-22.6128 -0.0005 -0.0003>, rotation=<rot 0 -89.9999 89.9992>}]
 calf_r [{shapeType=Capsule, length=35.255, radius=7.755, center=<22.6128 0.0005 0.0003>, rotation=<rot 0 -89.9999 -90.0008>}]
 clavicle_l [{shapeType=Capsule, length=9.505, radius=5.505, center=<5.0051 0.0002 0.0001>, rotation=<rot 0 -89.9999 -90.0008>}]
 clavicle_r [{shapeType=Capsule, length=9.505, radius=5.505, center=<-5.0051 -0.0002 -0.0001>, rotation=<rot 0 -89.9999 89.9992>}]
 dyn_backpack_COL [{shapeType=Capsule, length=12, radius=8.505, rotation=<rot 0 -89.9997 89.9995>}]
 dyn_ponytail_01 [{shapeType=Capsule, length=2.5, radius=3.5, center=<1.5 0 2.5>, rotation=<rot -75 180 -180>}]
 dyn_ponytail_01_MSTR [{shapeType=Capsule, length=0.001, radius=0.001, rotation=<rot 0 -90 0>}]
 dyn_ponytail_02 [{shapeType=Capsule, length=7, radius=4, center=<4.6492 0 0.0001>, rotation=<rot -90 -0.0198 0.0198>}]
 dyn_ponytail_03 [{shapeType=Capsule, length=7.1578, radius=4, center=<4.6492 0 0.0001>, rotation=<rot -90 -0.0198 0.0198>}]
 dyn_ponytail_04 [{shapeType=Capsule, length=7.1578, radius=4, center=<4.6492 0 0.0001>, rotation=<rot -90 -0.0198 0.0198>}]
 dyn_ponytail_05 [{shapeType=Capsule, length=7.1578, radius=4, center=<4.6492 0 0.0001>, rotation=<rot -90 -0.0198 0.0198>}]
 dyn_ponytail_06 [{shapeType=Capsule, length=7.1578, radius=4, center=<4.6492 0 0.0001>, rotation=<rot -90 -0.0198 0.0198>}]
 dyn_ponytail_07 [{shapeType=Capsule, length=7.1578, radius=4, center=<4.6492 0 0.0001>, rotation=<rot -90 -0.0198 0.0198>}]
 dyn_ponytail_08 [{shapeType=Capsule, length=7.1578, radius=4, center=<4.6492 0 0.0001>, rotation=<rot -90 -0.0198 0.0198>}]
 dyn_ponytail_09 [{shapeType=Capsule, length=7.1578, radius=4, center=<4.6492 0 0.0001>, rotation=<rot -90 -0.0198 0.0198>}]
 head [{shapeType=Capsule, length=7.755, radius=6.005, center=<4.0819 0 -0.0001>, rotation=<rot 0 -89.9999 -89.999>}]
 lowerarm_l [{shapeType=Capsule, length=17.505, radius=5.505, center=<11.3164 0.0001 0.0002>, rotation=<rot 0 -89.9999 -90.0008>}]
 lowerarm_r [{shapeType=Capsule, length=17.505, radius=5.505, center=<-11.3164 -0.0001 -0.0002>, rotation=<rot 0 -89.9999 89.9992>}]
 neck_01 [{shapeType=Capsule, length=8.255, radius=5.505, center=<4.0819 0 -0.0001>, rotation=<rot 0 -89.9999 -89.999>}]
 neck_02 [{shapeType=Capsule, length=6.505, radius=5.255, center=<4.0819 0 -0.0001>, rotation=<rot 0 -89.9999 -89.999>}]
 pelvis [{shapeType=Capsule, length=20.505, radius=7.505, rotation=<rot 0 -90 0>}]
 spine_01 [{shapeType=Capsule, length=20.505, radius=7.505, rotation=<rot 0 -90 0>}]
 spine_02 [{shapeType=Capsule, length=7.255, radius=8.255, rotation=<rot 0 -90 0>}]
 spine_03 [{shapeType=Capsule, length=8.005, radius=7.505, rotation=<rot 0 -90 0>}]
 spine_04 [{shapeType=Capsule, length=14.755, radius=7.505, center=<2.9669 -0.553 0>, rotation=<rot 0 -90 0>}]
 spine_05 [{shapeType=Capsule, length=19.255, radius=6.505, center=<4.3366 0.0003 0>, rotation=<rot 0 -90 0>}]
 thigh_l [{shapeType=Capsule, length=35.255, radius=7.755, center=<-22.6128 -0.0005 -0.0003>, rotation=<rot 0 -89.9999 89.9992>}]
 thigh_r [{shapeType=Capsule, length=35.255, radius=7.755, center=<22.6128 0.0005 0.0003>, rotation=<rot 0 -89.9999 -90.0008>}]
 upperarm_l [{shapeType=Capsule, length=17.505, radius=5.505, center=<11.3164 0.0001 0.0002>, rotation=<rot 0 -89.9999 -90.0008>}]
 upperarm_r [{shapeType=Capsule, length=17.505, radius=5.505, center=<-11.3164 -0.0001 -0.0002>, rotation=<rot 0 -89.9999 89.9992>}]
}
constraints [
 {
  bone1Name=dyn_ponytail_01_MSTR
  bone2Name=head
  swing1LimitDegrees=45
  swing1Motion=Free
  swing2LimitDegrees=45
  swing2Motion=Free
  twistLimitDegrees=45
  twistMotion=Free
 }
 {
  bone1Name=spine_02
  bone2Name=spine_01
  swing1LimitDegrees=45
  swing1Motion=Free
  swing2LimitDegrees=45
  swing2Motion=Free
  twistLimitDegrees=45
  twistMotion=Free
 }
 {
  bone1Name=spine_03
  bone2Name=spine_02
  swing1LimitDegrees=45
  swing1Motion=Free
  swing2LimitDegrees=45
  swing2Motion=Free
  twistLimitDegrees=45
  twistMotion=Free
 }
 {
  bone1Name=spine_04
  bone2Name=spine_03
  swing1LimitDegrees=45
  swing1Motion=Free
  swing2LimitDegrees=45
  swing2Motion=Free
  twistLimitDegrees=45
  twistMotion=Free
 }
 {
  bone1Name=spine_05
  bone2Name=spine_04
  swing1LimitDegrees=45
  swing1Motion=Free
  swing2LimitDegrees=45
  swing2Motion=Free
  twistLimitDegrees=45
  twistMotion=Free
 }
 {
  bone1Name=spine_01
  bone2Name=pelvis
  swing1LimitDegrees=45
  swing1Motion=Free
  swing2LimitDegrees=45
  swing2Motion=Free
  twistLimitDegrees=45
  twistMotion=Free
 }
 {
  bone1Name=thigh_l
  bone2Name=pelvis
  swing1LimitDegrees=45
  swing1Motion=Free
  swing2LimitDegrees=45
  swing2Motion=Free
  twistLimitDegrees=45
  twistMotion=Free
 }
 {
  bone1Name=calf_l
  bone2Name=thigh_l
  swing1LimitDegrees=45
  swing1Motion=Free
  swing2LimitDegrees=45
  swing2Motion=Free
  twistLimitDegrees=45
  twistMotion=Free
 }
 {
  bone1Name=thigh_r
  bone2Name=pelvis
  swing1LimitDegrees=45
  swing1Motion=Free
  swing2LimitDegrees=45
  swing2Motion=Free
  twistLimitDegrees=45
  twistMotion=Free
 }
 {
  bone1Name=calf_r
  bone2Name=thigh_r
  swing1LimitDegrees=45
  swing1Motion=Free
  swing2LimitDegrees=45
  swing2Motion=Free
  twistLimitDegrees=45
  twistMotion=Free
 }
 {
  bone1Name=neck_01
  bone2Name=spine_05
  swing1LimitDegrees=45
  swing1Motion=Free
  swing2LimitDegrees=45
  swing2Motion=Free
  twistLimitDegrees=45
  twistMotion=Free
 }
 {
  bone1Name=neck_02
  bone2Name=neck_01
  swing1LimitDegrees=45
  swing1Motion=Free
  swing2LimitDegrees=45
  swing2Motion=Free
  twistLimitDegrees=45
  twistMotion=Free
 }
 {
  bone1Name=head
  bone2Name=neck_02
  swing1LimitDegrees=45
  swing1Motion=Free
  swing2LimitDegrees=45
  swing2Motion=Free
  twistLimitDegrees=45
  twistMotion=Free
 }
 {
  bone1Name=clavicle_l
  bone2Name=spine_05
  swing1LimitDegrees=45
  swing1Motion=Free
  swing2LimitDegrees=45
  swing2Motion=Free
  twistLimitDegrees=45
  twistMotion=Free
 }
 {
  bone1Name=clavicle_r
  bone2Name=spine_05
  swing1LimitDegrees=45
  swing1Motion=Free
  swing2LimitDegrees=45
  swing2Motion=Free
  twistLimitDegrees=45
  twistMotion=Free
 }
 {bone1Name=dyn_ponytail_01, bone2Name=dyn_ponytail_01_MSTR, swing1LimitDegrees=10, swing2LimitDegrees=15, twistLimitDegrees=5}
 {bone1Name=dyn_ponytail_02, bone2Name=dyn_ponytail_01, swing1LimitDegrees=10, swing2LimitDegrees=15, twistLimitDegrees=5}
 {bone1Name=dyn_ponytail_03, bone2Name=dyn_ponytail_02, swing1LimitDegrees=10, swing2LimitDegrees=20, twistLimitDegrees=5}
 {bone1Name=dyn_ponytail_04, bone2Name=dyn_ponytail_03, swing1LimitDegrees=10, swing2LimitDegrees=20, twistLimitDegrees=5}
 {bone1Name=dyn_ponytail_05, bone2Name=dyn_ponytail_04, swing1LimitDegrees=10, swing2LimitDegrees=45, twistLimitDegrees=5}
 {bone1Name=dyn_ponytail_06, bone2Name=dyn_ponytail_05, swing1LimitDegrees=10, swing2LimitDegrees=45, twistLimitDegrees=5}
 {bone1Name=dyn_ponytail_07, bone2Name=dyn_ponytail_06, swing1LimitDegrees=10, swing2LimitDegrees=45, twistLimitDegrees=5}
 {bone1Name=dyn_ponytail_08, bone2Name=dyn_ponytail_07, swing1LimitDegrees=10, swing2LimitDegrees=45, twistLimitDegrees=5}
 {bone1Name=dyn_ponytail_09, bone2Name=dyn_ponytail_08, swing1LimitDegrees=10, swing2LimitDegrees=45, twistLimitDegrees=5}
 {
  bone1Name=upperarm_l
  bone2Name=clavicle_l
  swing1LimitDegrees=45
  swing1Motion=Free
  swing2LimitDegrees=45
  swing2Motion=Free
  twistLimitDegrees=45
  twistMotion=Free
 }
 {
  bone1Name=upperarm_r
  bone2Name=clavicle_r
  swing1LimitDegrees=45
  swing1Motion=Free
  swing2LimitDegrees=45
  swing2Motion=Free
  twistLimitDegrees=45
  twistMotion=Free
 }
 {
  bone1Name=dyn_backpack_COL
  bone2Name=spine_05
  swing1LimitDegrees=45
  swing1Motion=Free
  swing2LimitDegrees=45
  swing2Motion=Free
  twistLimitDegrees=45
  twistMotion=Free
 }
 {
  bone1Name=lowerarm_r
  bone2Name=upperarm_r
  swing1LimitDegrees=45
  swing1Motion=Free
  swing2LimitDegrees=45
  swing2Motion=Free
  twistLimitDegrees=45
  twistMotion=Free
 }
 {
  bone1Name=lowerarm_l
  bone2Name=upperarm_l
  swing1LimitDegrees=45
  swing1Motion=Free
  swing2LimitDegrees=45
  swing2Motion=Free
  twistLimitDegrees=45
  twistMotion=Free
 }
 {
  bone1Name=head
  bone2Name=dyn_ponytail_02
  swing1LimitDegrees=45
  swing1Motion=Free
  swing2LimitDegrees=45
  swing2Motion=Free
  twistLimitDegrees=45
  twistMotion=Free
 }
]
