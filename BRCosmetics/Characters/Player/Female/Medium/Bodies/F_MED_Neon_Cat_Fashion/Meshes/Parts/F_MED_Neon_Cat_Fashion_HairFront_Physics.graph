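# Physics asset for the front hair of the Neon Cat Fashion female medium cosmetic, covering left and right ponytail bone chains.
# Each chain links eight capsules to the head through a tiny MSTR anchor body, with narrow twist limits and swing tightening toward the tip.
# Capsules on torso, neck, head, arms and legs approximate the body, joined by angularly free constraints.
# The right arm uses the upperarm twist bone, and its forearm constrains directly to spine_05.
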PhysicsAsset /BRCosmetics/Characters/Player/Female/Medium/Bodies/F_MED_Neon_Cat_Fashion/Meshes/Parts/F_MED_Neon_Cat_Fashion_HairFront_Physics
bodies {
 calf_l [{shapeType=Capsule, length=35.255, radius=7.755, center=<-22.6128 -0.0005 -0.0003>, rotation=<rot 0 -89.9999 89.9992>}]
 calf_r [{shapeType=Capsule, length=35.255, radius=7.755, center=<22.6128 0.0005 0.0003>, rotation=<rot 0 -89.9999 -90.0008>}]
 dyn_ponytail_01_l [{shapeType=Capsule, length=6.6578, radius=2.25, center=<4.6492 0 0.0001>, rotation=<rot -90 -0.0198 0.0198>}]
 dyn_ponytail_01_l_MSTR [{shapeType=Capsule, length=0.001, radius=0.001, rotation=<rot 0 -90 0>}]
 dyn_ponytail_01_r [{shapeType=Capsule, length=6.6578, radius=2.25, center=<4.6492 0 0.0001>, rotation=<rot -90 -0.0198 0.0198>}]
 dyn_ponytail_01_r_MSTR [{shapeType=Capsule, length=0.001, radius=0.001, rotation=<rot 0 -90 0>}]
 dyn_ponytail_02_l [{shapeType=Capsule, length=6.4078, radius=2.5, center=<4.6492 0 0.0001>, rotation=<rot -90 -0.0198 0.0198>}]
 dyn_ponytail_02_r [{shapeType=Capsule, length=6.4078, radius=2.5, center=<4.6492 0 0.0001>, rotation=<rot -90 -0.0198 0.0198>}]
 dyn_ponytail_03_l [{shapeType=Capsule, length=6.4078, radius=2.5, center=<4.6492 0 0.0001>, rotation=<rot -90 -0.0198 0.0198>}]
 dyn_ponytail_03_r [{shapeType=Capsule, length=6.4078, radius=2.5, center=<4.6492 0 0.0001>, rotation=<rot -90 -0.0198 0.0198>}]
 dyn_ponytail_04_l [{shapeType=Capsule, length=8.6578, radius=2.5, center=<4.6492 0 0.0001>, rotation=<rot -90 -0.0198 0.0198>}]
 dyn_ponytail_04_r [{shapeType=Capsule, length=8.6578, radius=2.5, center=<4.6492 0 0.0001>, rotation=<rot -90 -0.0198 0.0198>}]
 dyn_ponytail_05_l [{shapeType=Capsule, length=8.6578, radius=2.5, center=<4.6492 0 0.0001>, rotation=<rot -90 -0.0198 0.0198>}]
 dyn_ponytail_05_r [{shapeType=Capsule, length=8.6578, radius=2.5, center=<4.6492 0 0.0001>, rotation=<rot -90 -0.0198 0.0198>}]
 dyn_ponytail_06_l [{shapeType=Capsule, length=8.6578, radius=2.5, center=<4.6492 0 0.0001>, rotation=<rot -90 -0.0198 0.0198>}]
 dyn_ponytail_06_r [{shapeType=Capsule, length=8.6578, radius=2.5, center=<4.6492 0 0.0001>, rotation=<rot -90 -0.0198 0.0198>}]
 dyn_ponytail_07_l [{shapeType=Capsule, length=8.6578, radius=2.5, center=<4.6492 0 0.0001>, rotation=<rot -90 -0.0198 0.0198>}]
 dyn_ponytail_07_r [{shapeType=Capsule, length=8.6578, radius=2.5, center=<4.6492 0 0.0001>, rotation=<rot -90 -0.0198 0.0198>}]
 dyn_ponytail_08_l [{shapeType=Capsule, length=8.6578, radius=2.5, center=<4.6492 0 0.0001>, rotation=<rot -90 -0.0198 0.0198>}]
 dyn_ponytail_08_r [{shapeType=Capsule, length=8.6578, radius=2.5, center=<4.6492 0 0.0001>, rotation=<rot -90 -0.0198 0.0198>}]
 head [{shapeType=Capsule, length=7.755, radius=6.005, center=<4.0819 0 -0.0001>, rotation=<rot 0 -89.9999 -89.999>}]
 lowerarm_l [{shapeType=Capsule, length=38.005, radius=3.755, center=<19.4584 0.0008 0.0004>, rotation=<rot 0 -89.9999 -90.0008>}]
 lowerarm_r [{shapeType=Capsule, length=38.005, radius=3.755, center=<-19.4584 -0.0004 -0.0004>, rotation=<rot 0 -89.9999 89.9992>}]
 neck_01 [{shapeType=Capsule, length=8.255, radius=5.505, center=<4.0819 0 -0.0001>, rotation=<rot 0 -89.9999 -89.999>}]
 neck_02 [{shapeType=Capsule, length=6.505, radius=5.255, center=<4.0819 0 -0.0001>, rotation=<rot 0 -89.9999 -89.999>}]
 pelvis [{shapeType=Capsule, length=20.505, radius=7.505, rotation=<rot 0 -90 0>}]
 spine_01 [{shapeType=Capsule, length=20.505, radius=7.505, rotation=<rot 0 -90 0>}]
 spine_02 [{shapeType=Capsule, length=7.255, radius=8.255, rotation=<rot 0 -90 0>}]
 spine_03 [{shapeType=Capsule, length=35.005, radius=7.505, rotation=<rot 0 -90 0>}]
 spine_04 [
  {shapeType=Capsule, length=13.255, radius=9.005, center=<2.9669 -0.553 0>, rotation=<rot 0 -90 0>}
  {shapeType=Capsule, length=24.75, radius=5.25, center=<0 7.0868 0>}
 ]
 spine_05 [
  {shapeType=Capsule, length=27.505, radius=6.505, center=<4.3366 0.0003 0>, rotation=<rot 0 -90 0>}
  {shapeType=Capsule, length=79.25, radius=15, center=<13.9993 -15.0221 0>}
 ]
 thigh_l [{shapeType=Capsule, length=33.255, radius=9.755, center=<-22.6128 -0.0005 -0.0003>, rotation=<rot 0 -89.9999 89.9992>}]
 thigh_r [{shapeType=Capsule, length=33.255, radius=9.755, center=<22.6128 0.0005 0.0003>, rotation=<rot 0 -89.9999 -90.0008>}]
 upperarm_l [{shapeType=Capsule, length=27.005, radius=3.755, center=<12.7374 0.0002 0.0003>, rotation=<rot 0 -89.9999 -90.0008>}]
 upperarm_twist_01_r [{shapeType=Capsule, length=27.005, radius=3.755, center=<-4.1494 -0.0006 -0.0002>, rotation=<rot 0 -89.9999 89.9992>}]
}
constraints [
 {bone1Name=dyn_ponytail_01_l, bone2Name=dyn_ponytail_01_l_MSTR, swing1LimitDegrees=60, swing2LimitDegrees=60, twistLimitDegrees=5}
 {bone1Name=dyn_ponytail_02_l, bone2Name=dyn_ponytail_01_l, swing1LimitDegrees=60, swing2LimitDegrees=60, twistLimitDegrees=5}
 {bone1Name=dyn_ponytail_03_l, bone2Name=dyn_ponytail_02_l, swing1LimitDegrees=60, swing2LimitDegrees=60, twistLimitDegrees=8}
 {bone1Name=dyn_ponytail_04_l, bone2Name=dyn_ponytail_03_l, swing1LimitDegrees=45, swing2LimitDegrees=45, twistLimitDegrees=8}
 {bone1Name=dyn_ponytail_05_l, bone2Name=dyn_ponytail_04_l, swing1LimitDegrees=45, swing2LimitDegrees=45, twistLimitDegrees=8}
 {bone1Name=dyn_ponytail_06_l, bone2Name=dyn_ponytail_05_l, swing1LimitDegrees=45, swing2LimitDegrees=45, twistLimitDegrees=8}
 {bone1Name=dyn_ponytail_07_l, bone2Name=dyn_ponytail_06_l, swing1LimitDegrees=45, swing2LimitDegrees=45, twistLimitDegrees=8}
 {bone1Name=dyn_ponytail_08_l, bone2Name=dyn_ponytail_07_l, swing1LimitDegrees=45, swing2LimitDegrees=45, twistLimitDegrees=8}
 {bone1Name=dyn_ponytail_01_r, bone2Name=dyn_ponytail_01_r_MSTR, swing1LimitDegrees=60, swing2LimitDegrees=60, twistLimitDegrees=5}
 {bone1Name=dyn_ponytail_02_r, bone2Name=dyn_ponytail_01_r, swing1LimitDegrees=60, swing2LimitDegrees=60, twistLimitDegrees=5}
 {bone1Name=dyn_ponytail_03_r, bone2Name=dyn_ponytail_02_r, swing1LimitDegrees=60, swing2LimitDegrees=60, twistLimitDegrees=8}
 {bone1Name=dyn_ponytail_04_r, bone2Name=dyn_ponytail_03_r, swing1LimitDegrees=45, swing2LimitDegrees=45, twistLimitDegrees=8}
 {bone1Name=dyn_ponytail_05_r, bone2Name=dyn_ponytail_04_r, swing1LimitDegrees=45, swing2LimitDegrees=45, twistLimitDegrees=8}
 {bone1Name=dyn_ponytail_06_r, bone2Name=dyn_ponytail_05_r, swing1LimitDegrees=45, swing2LimitDegrees=45, twistLimitDegrees=8}
 {bone1Name=dyn_ponytail_07_r, bone2Name=dyn_ponytail_06_r, swing1LimitDegrees=45, swing2LimitDegrees=45, twistLimitDegrees=8}
 {bone1Name=dyn_ponytail_08_r, bone2Name=dyn_ponytail_07_r, swing1LimitDegrees=45, swing2LimitDegrees=45, twistLimitDegrees=8}
 {
  bone1Name=dyn_ponytail_01_l_MSTR
  bone2Name=head
  swing1LimitDegrees=45
  swing1Motion=Free
  swing2LimitDegrees=45
  swing2Motion=Free
  twistLimitDegrees=45
  twistMotion=Free
 }
 {
  bone1Name=dyn_ponytail_01_r_MSTR
  bone2Name=head
  swing1LimitDegrees=45
  swing1Motion=Free
  swing2LimitDegrees=45
  swing2Motion=Free
  twistLimitDegrees=45
  twistMotion=Free
 }
 {
  bone1Name=spine_02
  bone2Name=spine_01
  swing1LimitDegrees=45
  swing1Motion=Free
  swing2LimitDegrees=45
  swing2Motion=Free
  twistLimitDegrees=45
  twistMotion=Free
 }
 {
  bone1Name=spine_03
  bone2Name=spine_02
  swing1LimitDegrees=45
  swing1Motion=Free
  swing2LimitDegrees=45
  swing2Motion=Free
  twistLimitDegrees=45
  twistMotion=Free
 }
 {
  bone1Name=spine_04
  bone2Name=spine_03
  swing1LimitDegrees=45
  swing1Motion=Free
  swing2LimitDegrees=45
  swing2Motion=Free
  twistLimitDegrees=45
  twistMotion=Free
 }
 {
  bone1Name=spine_05
  bone2Name=spine_04
  swing1LimitDegrees=45
  swing1Motion=Free
  swing2LimitDegrees=45
  swing2Motion=Free
  twistLimitDegrees=45
  twistMotion=Free
 }
 {
  bone1Name=spine_01
  bone2Name=pelvis
  swing1LimitDegrees=45
  swing1Motion=Free
  swing2LimitDegrees=45
  swing2Motion=Free
  twistLimitDegrees=45
  twistMotion=Free
 }
 {
  bone1Name=thigh_l
  bone2Name=pelvis
  swing1LimitDegrees=45
  swing1Motion=Free
  swing2LimitDegrees=45
  swing2Motion=Free
  twistLimitDegrees=45
  twistMotion=Free
 }
 {
  bone1Name=calf_l
  bone2Name=thigh_l
  swing1LimitDegrees=45
  swing1Motion=Free
  swing2LimitDegrees=45
  swing2Motion=Free
  twistLimitDegrees=45
  twistMotion=Free
 }
 {
  bone1Name=thigh_r
  bone2Name=pelvis
  swing1LimitDegrees=45
  swing1Motion=Free
  swing2LimitDegrees=45
  swing2Motion=Free
  twistLimitDegrees=45
  twistMotion=Free
 }
 {
  bone1Name=calf_r
  bone2Name=thigh_r
  swing1LimitDegrees=45
  swing1Motion=Free
  swing2LimitDegrees=45
  swing2Motion=Free
  twistLimitDegrees=45
  twistMotion=Free
 }
 {
  bone1Name=neck_01
  bone2Name=spine_05
  swing1LimitDegrees=45
  swing1Motion=Free
  swing2LimitDegrees=45
  swing2Motion=Free
  twistLimitDegrees=45
  twistMotion=Free
 }
 {
  bone1Name=neck_02
  bone2Name=neck_01
  swing1LimitDegrees=45
  swing1Motion=Free
  swing2LimitDegrees=45
  swing2Motion=Free
  twistLimitDegrees=45
  twistMotion=Free
 }
 {
  bone1Name=head
  bone2Name=neck_02
  swing1LimitDegrees=45
  swing1Motion=Free
  swing2LimitDegrees=45
  swing2Motion=Free
  twistLimitDegrees=45
  twistMotion=Free
 }
 {
  bone1Name=upperarm_l
  bone2Name=spine_05
  swing1LimitDegrees=45
  swing1Motion=Free
  swing2LimitDegrees=45
  swing2Motion=Free
  twistLimitDegrees=45
  twistMotion=Free
 }
 {
  bone1Name=upperarm_twist_01_r
  bone2Name=spine_05
  swing1LimitDegrees=45
  swing1Motion=Free
  swing2LimitDegrees=45
  swing2Motion=Free
  twistLimitDegrees=45
  twistMotion=Free
 }
 {
  bone1Name=lowerarm_r
  bone2Name=spine_05
  swing1LimitDegrees=45
  swing1Motion=Free
  swing2LimitDegrees=45
  swing2Motion=Free
  twistLimitDegrees=45
  twistMotion=Free
 }
 {
  bone1Name=lowerarm_l
  bone2Name=upperarm_l
  swing1LimitDegrees=45
  swing1Motion=Free
  swing2LimitDegrees=45
  swing2Motion=Free
  twistLimitDegrees=45
  twistMotion=Free
 }
]
