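# Post-copy Anim Blueprint for the Neon Cat Fashion body part that adds secondary motion to hair and ponytails.
# The pose copied from the parent mesh is aim-corrected and cached, then bRigidBodyEnabled selects front/back RigidBody simulation or a Look At plus Trail Controller fallback.
# The event graph derives trail alphas from movement and gravity from skydiving force.
# Backbling and cape checks drive additive bone offsets that adjust the hair.
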AnimBlueprint /BRCosmetics/Characters/Player/Female/Medium/Bodies/F_MED_Neon_Cat_Fashion/Meshes/Parts/F_MED_Neon_Cat_Fashion_AnimBP
target_skeleton=&/F_MED_Neon_Cat_Fashion_Skeleton
graphs [
 AnimGraph {nodes [
   n1 = "Transform (Modify) Bone - Bone: dyn_ponytail_02_l"(Alpha=1, Translation="0,0,0", Rotation="24.995,0.401,-0.069", Scale="1,1,1") {
    rotationMode=BMM_Additive
    rotationSpace=BCS_BoneSpace
    translationMode=BMM_Additive
    translationSpace=BCS_BoneSpace
    alphaBoolBlend {blendOption=Linear}
    boneToModify {boneName=dyn_ponytail_02_l}
    rotation=<rot 24.995 0.401 -0.069>
   }
   "Transform (Modify) Bone - Bone: dyn_ponytail_02_r"(ComponentPose=@n1, Alpha=1, Translation="0,0,0", Rotation="24.995,0.401,-0.069", Scale="1,1,1") {
    rotationMode=BMM_Additive
    rotationSpace=BCS_BoneSpace
    translationMode=BMM_Additive
    translationSpace=BCS_BoneSpace
    alphaBoolBlend {blendOption=Linear}
    boneToModify {boneName=dyn_ponytail_02_r}
    rotation=<rot 24.995 0.401 -0.069>
   }
   n2 = "Transform (Modify) Bone - Bone: dyn_ponytail_01_MSTR"(ComponentPose=Local To Component(LocalPose=Use cached pose 'prePose'()), Alpha=1, Translation="0,0,0", Rotation="15,0,0", Scale="1,1,1") {
    rotationMode=BMM_Additive
    rotationSpace=BCS_BoneSpace
    alphaBoolBlend {blendOption=Linear}
    boneToModify {boneName=dyn_ponytail_01_MSTR}
    rotation {pitch=15}
   }
   n3 = "Transform (Modify) Bone - Bone: dyn_backpack_COL"(
    ComponentPose=@n2
    Alpha=Get isWearingBackbling() -> (isWearingBackbling:double)
    Translation="-0.000,-93.405,0.000"
    Rotation="0,0,0"
    Scale="1,1,1"
   ) {
    rotationMode=BMM_Additive
    rotationSpace=BCS_BoneSpace
    scaleMode=BMM_Additive
    scaleSpace=BCS_BoneSpace
    translationMode=BMM_Additive
    translationSpace=BCS_BoneSpace
    alphaBoolBlend {blendOption=Linear}
    boneToModify {boneName=dyn_backpack_COL}
    translation {y=-93.405}
   }
   n4 = Get gravityVector() -> (gravityVector:Vector)
   n5 = RigidBody!(ComponentPose=@n3, Alpha=Get skydivingAlpha() -> (skydivingAlpha:double), OverrideWorldGravity:Vector=@n4, ExternalForce) -> (Pose) {
    bClampLinearTranslationLimitToRefPose=true
    bFreezeIncomingPoseOnStart=true
    bOverrideWorldGravity=true
    overridePhysicsAsset=&/F_MED_Neon_Cat_Fashion_HairFront_Physics
    simulationSpace=BaseBoneSpace
    alphaBoolBlend {blendOption=Linear}
    baseBoneRef {boneName=spine_05}
    componentAppliedLinearAccClamp=<1200 1200 1200>
    componentLinearAccScale=<0.2 0.2 0.2>
    componentLinearVelScale=<0.2 0.2 0.2>
    overrideWorldGravity {z=-980}
   }
   n6 = RigidBody(ComponentPose=@n5, OverrideWorldGravity=@n4) {
    bClampLinearTranslationLimitToRefPose=true
    bFreezeIncomingPoseOnStart=true
    bOverrideWorldGravity=true
    overridePhysicsAsset=&/F_MED_Neon_Cat_Fashion_HairBack_Physics
    simulationSpace=BaseBoneSpace
    alphaBoolBlend {blendOption=Linear}
    baseBoneRef {boneName=spine_05}
    componentAppliedLinearAccClamp=<1200 1200 1200>
    componentLinearAccScale=<0.15 0.15 0.2>
    componentLinearVelScale=<0.15 0.15 0.2>
    overrideWorldGravity {z=-980}
   }
   n7 = Component To Local(ComponentPose=@n6)
   n8 = Look At - Bone: dyn_ponytail_01_l(ComponentPose=Local To Component(LocalPose=Use cached pose 'prePose'()), Alpha=1) {
    alphaBoolBlend {blendOption=Linear}
    boneToModify {boneName=dyn_ponytail_01_l}
    lookAtLocation=<0 10 -10>
    lookAtTarget {boneReference {boneName=spine_04}}
    lookAt_Axis {axis=<1 0>}
   }
   n9 = Look At - Bone: dyn_ponytail_03_l(ComponentPose=@n8, Alpha=1) {
    alphaBoolBlend {blendOption=Linear}
    boneToModify {boneName=dyn_ponytail_03_l}
    lookAtLocation=<10 55>
    lookAtTarget {boneReference {boneName=root}}
    lookAt_Axis {axis=<1 0>}
   }
   n10 = Look At - Bone: dyn_ponytail_01_r(ComponentPose=@n9, Alpha=1) {
    alphaBoolBlend {blendOption=Linear}
    boneToModify {boneName=dyn_ponytail_01_r}
    lookAtLocation=<0 10 10>
    lookAtTarget {boneReference {boneName=spine_04}}
    lookAt_Axis {axis=<1 0>}
   }
   n11 = Look At - Bone: dyn_ponytail_03_r(ComponentPose=@n10, Alpha=1) {
    alphaBoolBlend {blendOption=Linear}
    boneToModify {boneName=dyn_ponytail_03_r}
    lookAtLocation=<-10 55>
    lookAtTarget {boneReference {boneName=root}}
    lookAt_Axis {axis=<1 0>}
   }
   n12 = Look At - Bone: dyn_ponytail_01_TARGET(ComponentPose=@n11, Alpha=1) {
    alphaBoolBlend {blendOption=Linear}
    boneToModify {boneName=dyn_ponytail_01_TARGET}
    lookAtLocation {x=0}
    lookAtTarget {boneReference {boneName=dyn_ponytail_01_AIM}}
    lookAt_Axis {axis=<1 0>}
   }
   n13 = Look At - Bone: dyn_ponytail_02(ComponentPose=@n12, Alpha=1) {
    alphaBoolBlend {blendOption=Linear}
    boneToModify {boneName=dyn_ponytail_02}
    lookAtLocation=<0 -40>
    lookAtTarget {boneReference {boneName=pelvis}}
    lookAt_Axis {axis=<1 0>}
   }
   n14 = Look At - Bone: dyn_ponytail_03(ComponentPose=@n13, Alpha=1) {
    alphaBoolBlend {blendOption=Linear}
    boneToModify {boneName=dyn_ponytail_03}
    lookAtLocation=<0 -25>
    lookAtTarget {boneReference {boneName=pelvis}}
    lookAt_Axis {axis=<1 0>}
   }
   n15 = Trail controller - Bone: dyn_ponytail_09!(ComponentPose=@n14, Alpha=Get BwdMovement() -> (BwdMovement:double), RelaxationSpeedScale:float=1, FakeVelocity:Vector="0,0,0") -> (Pose) {
    chainLength=8
    maxDeltaTime=0.0333
    relaxationSpeedScale=1.2
    alphaScaleBiasClamp {bInterpResult=true}
    baseJoint {boneName=root}
    planarLimits [
     {drivingBone {boneName=thigh_l}, planeTransform {location=<-20 -8 0>, rotation=<rot 0 0 -90>, scale=<0.5 0.5 0.5>}}
     {drivingBone {boneName=thigh_r}, planeTransform {location=<20 8 0>, rotation=<rot 0 0 90>, scale=<0.5 0.5 0.5>}}
    ]
    rotationLimits [
     {limitMax=<0 0 30>, limitMin=<0 -30 -30>}
     {limitMax=<0 0 30>, limitMin=<0 -30 -30>}
     {limitMax=<0 0 30>, limitMin=<0 -30 -30>}
     {limitMax=<180 180 180>}
     {limitMax=<180 180 180>}
     {limitMax=<180 180 180>}
     {limitMax=<180 180 180>}
     {limitMax=<180 180 180>}
    ]
    rotationOffsets [<0 0 0>, <0 0 0>, <0 0 0>, <0 0 0>, <0 0 0>, <0 0 0>, <0 0 0>, <0 0 0>]
    trailBone {boneName=dyn_ponytail_09}
    trailRelaxationSpeed {editorCurveData {keys [<at 0 32 RCIM_Linear>,<at 0.3 32 RCIM_Linear>,<at 1 10 RCIM_Linear>]}}
   }
   n16 = Get FwdMovement() -> (FwdMovement:double)
   n17 = Trail controller - Bone: dyn_ponytail_08_l!(ComponentPose=@n15, Alpha=@n16, RelaxationSpeedScale:float=1.5, FakeVelocity:Vector="0,0,0") -> (Pose) {
    chainLength=7
    maxDeltaTime=0.0333
    relaxationSpeedScale=1.2
    alphaScaleBiasClamp {bInterpResult=true}
    baseJoint {boneName=root}
    planarLimits [
     {drivingBone {boneName=root}, planeTransform {location=<0 0 10>, rotation=<rot 0 0 0>, scale=<1 1 1>}}
     {drivingBone {boneName=thigh_l}, planeTransform {location=<-20 6 0>, rotation=<rot 0 0 90>, scale=<0.5 0.5 0.5>}}
    ]
    rotationLimits [
     {limitMax=<0 0 0>, limitMin=<0 -30 0>}
     {limitMax=<0 0 5>, limitMin=<0 -30 -5>}
     {limitMax=<0 0 15>, limitMin=<0 -30 -15>}
     {limitMax=<0 0 15>, limitMin=<0 -30 -15>}
     {limitMax=<180 180 180>}
     {limitMax=<180 180 180>}
     {limitMax=<180 180 180>}
    ]
    rotationOffsets [<0 0 0>, <0 0 0>, <0 0 0>, <0 0 0>, <0 0 0>, <0 0 0>, <0 0 0>]
    trailBone {boneName=dyn_ponytail_08_l}
    trailRelaxationSpeed {editorCurveData {keys [<at 0 32 RCIM_Linear>,<at 0.3 32 RCIM_Linear>,<at 1 10 RCIM_Linear>]}}
   }
   n18 = Trail controller - Bone: dyn_ponytail_08_r!(ComponentPose=@n17, Alpha=@n16, RelaxationSpeedScale:float=1.5, FakeVelocity:Vector="0,0,0") -> (Pose) {
    chainLength=7
    maxDeltaTime=0.0333
    relaxationSpeedScale=1.2
    alphaScaleBiasClamp {bInterpResult=true}
    baseJoint {boneName=root}
    planarLimits [
     {drivingBone {boneName=root}, planeTransform {location=<0 0 10>, rotation=<rot 0 0 0>, scale=<1 1 1>}}
     {drivingBone {boneName=thigh_r}, planeTransform {location=<20 -6 0>, rotation=<rot 0 0 -90>, scale=<0.5 0.5 0.5>}}
    ]
    rotationLimits [
     {limitMax=<0 0 0>, limitMin=<0 -30 0>}
     {limitMax=<0 0 5>, limitMin=<0 -30 -5>}
     {limitMax=<0 0 15>, limitMin=<0 -30 -15>}
     {limitMax=<0 0 15>, limitMin=<0 -30 -15>}
     {limitMax=<180 180 180>}
     {limitMax=<180 180 180>}
     {limitMax=<180 180 180>}
    ]
    rotationOffsets [<0 0 0>, <0 0 0>, <0 0 0>, <0 0 0>, <0 0 0>, <0 0 0>, <0 0 0>]
    trailBone {boneName=dyn_ponytail_08_r}
    trailRelaxationSpeed {editorCurveData {keys [<at 0 32 RCIM_Linear>,<at 0.3 32 RCIM_Linear>,<at 1 10 RCIM_Linear>]}}
   }
   n19 = Get isWearingBackbling()
   n20 = "Transform (Modify) Bone - Bone: dyn_ponytail_01"(ComponentPose=@n18, Alpha=@n19, Translation="0,0,0", Rotation="-3.976,0.000,0.000", Scale="1,1,1") {
    rotationMode=BMM_Additive
    rotationSpace=BCS_BoneSpace
    scaleMode=BMM_Additive
    scaleSpace=BCS_BoneSpace
    translationMode=BMM_Additive
    translationSpace=BCS_BoneSpace
    alphaBoolBlend {blendOption=Linear}
    boneToModify {boneName=dyn_ponytail_01}
    rotation {pitch=-3.976}
   }
   n21 = "Transform (Modify) Bone - Bone: dyn_ponytail_02"(ComponentPose=@n20, Alpha=@n19, Translation="0,0,0", Rotation="19.918,0.000,0.000", Scale="1,1,1") {
    rotationMode=BMM_Additive
    rotationSpace=BCS_BoneSpace
    scaleMode=BMM_Additive
    scaleSpace=BCS_BoneSpace
    translationMode=BMM_Additive
    translationSpace=BCS_BoneSpace
    alphaBoolBlend {blendOption=Linear}
    boneToModify {boneName=dyn_ponytail_02}
    rotation {pitch=19.918}
   }
   n22 = "Transform (Modify) Bone - Bone: dyn_ponytail_03"(ComponentPose=@n21, Alpha=@n19, Translation="0,0,0", Rotation="-18.517,0.000,0.000", Scale="1,1,1") {
    rotationMode=BMM_Additive
    rotationSpace=BCS_BoneSpace
    scaleMode=BMM_Additive
    scaleSpace=BCS_BoneSpace
    translationMode=BMM_Additive
    translationSpace=BCS_BoneSpace
    alphaBoolBlend {blendOption=Linear}
    boneToModify {boneName=dyn_ponytail_03}
    rotation {pitch=-18.517}
   }
   n23 = Component To Local(ComponentPose=@n22)
   n24 = Get bRigidBodyEnabled()
   n25 = "Transform (Modify) Bone - Bone: dyn_ponytail_01"(
    ComponentPose=Local To Component(LocalPose=Blend Poses by bool(BlendPose_0=@n7, BlendPose_1=@n23, bActiveValue=@n24) {blendType=Linear})
    Alpha=Get isWearingCape() -> (isWearingCape:double)
    Translation="0,0,0"
    Rotation="15,0.000,0.000"
    Scale="1,1,1"
   ) {
    rotationMode=BMM_Additive
    rotationSpace=BCS_BoneSpace
    alphaBoolBlend {blendOption=Linear}
    boneToModify {boneName=dyn_ponytail_01}
    rotation {pitch=15}
   }
   Output Pose(Result=Component To Local(ComponentPose=@n25)) {layerGroup=None}
   n26 = Look At - Bone: dyn_ponytail_01_l_MSTR(ComponentPose=Local To Component(LocalPose=Copy Pose From Mesh() {bUseAttachedParent=true}), Alpha=1) {
    alphaBoolBlend {blendOption=Linear}
    boneToModify {boneName=dyn_ponytail_01_l_MSTR}
    lookAtLocation {x=0}
    lookAtTarget {boneReference {boneName=dyn_ponytail_01_l_AIM}}
    lookAt_Axis {axis=<1 0>}
    lookUp_Axis {axis {x=0, z=1}}
   }
   n27 = Look At - Bone: dyn_ponytail_01_r_MSTR(ComponentPose=@n26, Alpha=1) {
    alphaBoolBlend {blendOption=Linear}
    boneToModify {boneName=dyn_ponytail_01_r_MSTR}
    lookAtLocation {x=0}
    lookAtTarget {boneReference {boneName=dyn_ponytail_01_r_AIM}}
    lookAt_Axis {axis=<1 0>}
    lookUp_Axis {axis {x=0, z=1}}
   }
   Save cached pose 'prePose'(Pose=Component To Local(ComponentPose=@n27)) {cacheName=prePose}
  ]}
 EventGraph {nodes [
   Event BlueprintUpdateAnimation()
   > n1 = Sequence!(execute) -> (then_0, then_1, then_2:Exec)
   n2 = Break Fort Anim Input Skydiving External Force(FortAnimInput_SkydivingExternalForce=Get SkydivingExternalForceInput()) -> (
    bUseSkydivingVectorForce=~
    SkydivingVectorForce=~
    HeadToPelvisDirection=~
    FloatingMultiplier=~
    FloatingAdditive=~
    DivingMultiplier=~
    DivingAdditive=~
    ParachutingMultiplier=~
    ParachutingAdditive=~
   )
   n3 = Get bIsSkyDiving()
   > Set gravityVector(gravityVector:Vector=SelectVector(A=vector * float(A=@n2.SkydivingVectorForce, B=0.15), B="0, 0,-980", bPickA=@n3)) -> (Output_Get:Vector)
   > Set skydivingAlpha(skydivingAlpha:double=SelectFloat(A=0.25, B=1, bPickA=@n3)) -> (Output_Get:double)
   n4 = PlayerMovement(execute=@n1.then_1, AnimInstance=Self-Reference(), XDotProductCoefficient=1, YDotProductCoefficient=1, ZDotProductCoefficient=1)
   n5 = MapRangeClamped(Value=@n4.YMovement, InRangeA=0, InRangeB=1)
   > Set FwdMovement(FwdMovement:double=MapRangeClamped(Value=float + float(A=@n4, B=@n5) {is=title K2Node_CommutativeAssociativeBinaryOperator}, InRangeA=0, InRangeB=1, OutRangeA=0.5, OutRangeB=0)) -> (Output_Get:double)
   n6 = MapRangeClamped(Value=@n4.YMovement, InRangeA=0, InRangeB=-1)
   n7 = "Absolute (Float)"(A=@n4.ZMovement)
   > Set BwdMovement(BwdMovement:double=MapRangeClamped(Value=float + float(A=@n7, B=@n6), InRangeA=0, InRangeB=1.5, OutRangeA=1, OutRangeB=0)) -> (Output_Get:double)
   Event OnCharacterCustomizationLoaded()
   > Do Once()
   n8 = CheckForBackbling()
   n9 = NOT Boolean(A=@n8.bIsBackblingLong)
   > Set isWearingBackbling(isWearingBackbling:double="To Float (Boolean)"(InBool=AND Boolean(A=@n8, B=@n9))) -> (Output_Get:double=0)
   > Set isWearingCape(isWearingCape:double="To Float (Boolean)"(InBool=@n8.bIsBackblingLong)) -> (Output_Get:double)
  ]}
]
variables [
 Body Mesh {type=SkeletalMeshComponent}
 BwdMovement {type=number}
 FwdMovement {type=number}
 gravityVector {type=Vector}
 Head Mesh {type=SkeletalMeshComponent}
 IsCrouching {type=boolean}
 isWearingBackbling {type=number}
 isWearingCape {type=number}
 NewVar_0 {type=boolean}
 Pawn Owner {type=PlayerPawn_Athena_C}
 skydivingAlpha {type=number}
]
